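# Translucent deferred decal material for the building blocker miniboss effect's ground, combining a tinted rock texture with a rift decal mask.
# A normal derived from the rift mask's height channel yields an edge term (absolute red plus green) that is powered, colored and sphere-masked for emissive glow.
# The rift mask is also sampled clamped with center scaling, providing an inverted red channel and a powered green channel for opacity.
# Parameters expose tiling, scale, colors, exponents and surface response.
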Material /Game/Effects/Fort_Effects/Miniboss/BuildingBlocker/MD_BuildingBlockerGround
graph {
 properties {blendMode=BLEND_Translucent, materialDomain=MD_DeferredDecal, pixelDepthOffsetMode=PDOM_Legacy, usedShadingModels=Default Lit}
 nodes [
  n1 = Multiply(A=TextureCoordinate(), B=ScalarParameter() {defaultValue=1, parameterName=UVScale})
  n2 = MaterialFunctionCall!(
   Height Map=TextureObject() {samplerType=SAMPLERTYPE_Masks, texture=/Game/Missions/Primary/DeliverTheBomb/Textures/rift_ground_decal}
   Normal Map Intensity=ScalarParameter() {defaultValue=2, parameterName=NormalMapIntensity}
   Height Map UV Offset
   Coordinates=@n1
   Height Map Channel Selector=Constant4Vector() {constant=<rgba 1 0 0 0>}.RGBA
  ) {materialFunction=/Engine/Functions/Engine_MaterialFunctions03/Procedurals/NormalFromHeightmap}
  n3 = Abs(None=Add(A=ComponentMask(None=@n2.Result) {r=true}, B=ComponentMask(None=@n2.Result) {g=true}))
  n4 = ScalarParameter() {defaultValue=1, parameterName=Emissive Exponent}
  Multiply(
   A=Multiply(A=Power(Base=@n3, Exp=@n4), B=VectorParameter() {parameterName=Emissive Color}.RGB)
   B=SphereMask(
    A=TextureCoordinate()
    B=Constant() {r=0.5}
    Radius=ScalarParameter() {defaultValue=0.4, parameterName=SphereMaskRadius}
    Hardness=ScalarParameter() {parameterName=SphereMaskHardness}
   )
  )
  Multiply(
   A=TextureSample(UVs=@n1) {texture=/Game/Missions/Primary/DeliverTheBomb/Textures/T_Wasteland_RockWaveey}.RGB
   B=VectorParameter() {parameterName=BaseColorTint, defaultValue=<rgba 0.115 0.1 0.25 1>}.RGB
  )
  n5 = TextureSample(UVs=MaterialFunctionCall(UVs=TextureCoordinate(), Texture Scale=ScalarParameter() {defaultValue=1, parameterName=Scale}) {materialFunction=&ScaleUVsByCenter}.UVs) {
   samplerSource=SSM_Clamp_WorldGroupSettings
   samplerType=SAMPLERTYPE_Masks
   texture=&rift_ground_decal
  }
  OneMinus(None=@n5.R)
  Power(Base=@n5.G, Exp=ScalarParameter() {defaultValue=1, parameterName=Opacity Exponent})
  ScalarParameter() {parameterName=Metallic}
  ScalarParameter() {defaultValue=0.3, parameterName=Specular}
  ScalarParameter() {defaultValue=0.6, parameterName=Roughness}
 ]
}
